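# Sound cue for ambient swells inside the Arcade spot of the Hollow Base Takeover POI.
# A Random node picks one of seven equally weighted swell waves, and a Modulator above it adds a slight upward pitch variation.
# It references the shared tile-ambient one-shot attenuation, uses the outdoor ambient sound class despite its interior name, and has bus sends disabled.
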SoundCue /Game/Sounds/Artemis/POI/Hollow_Base_Takover/Spots/Arcade/Spot_Arcade_Interior_Ambient_Swells_Cue
root=Modulator_0
nodes [
 {id=Modulator_0, children [Random_1], properties {pitchMax=1.1, pitchMin=1}}
 {
  id=Random_1
  children [WavePlayer_13, WavePlayer_14, WavePlayer_15, WavePlayer_16, WavePlayer_17, WavePlayer_18, WavePlayer_19]
  properties {weights [1,1,1,1,1,1,1]}
 }
 {
  id=WavePlayer_13
  properties {soundWaveAssetPtr=&/Spot_Arcade_Interior_Ambient_Swells_02}
 }
 {
  id=WavePlayer_14
  properties {soundWaveAssetPtr=&/Spot_Arcade_Interior_Ambient_Swells_03}
 }
 {
  id=WavePlayer_15
  properties {soundWaveAssetPtr=&/Spot_Arcade_Interior_Ambient_Swells_04}
 }
 {
  id=WavePlayer_16
  properties {soundWaveAssetPtr=&/Spot_Arcade_Interior_Ambient_Swells_05}
 }
 {
  id=WavePlayer_17
  properties {soundWaveAssetPtr=&/Spot_Arcade_Interior_Ambient_Swells_06}
 }
 {
  id=WavePlayer_18
  properties {soundWaveAssetPtr=&/Spot_Arcade_Interior_Ambient_Swells_07}
 }
 {
  id=WavePlayer_19
  properties {soundWaveAssetPtr=&/Spot_Arcade_Interior_Ambient_Swells_08}
 }
]
properties {
 attenuationSettings=/Game/Sounds/Attenuation/Fort_TileAmbient_OneShot_Att
 bEnableBusSends=false
 maxDistance=8000
 soundClassObject=&Fort_Ambient_Outdoors_Athena
 volumeMultiplier=0.877
 attenuationOverrides {falloffDistance=6000, stereoSpread=10000, attenuationShapeExtents {x=2000}}
}
