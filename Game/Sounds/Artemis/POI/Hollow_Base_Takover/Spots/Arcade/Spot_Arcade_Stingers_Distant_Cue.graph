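# Sound cue for distant arcade stinger one-shots, an outdoor ambient spot sound in the Hollow Base Takover point of interest.
# A Random node picks one of twelve stinger waves with near-equal weights, and a root Modulator slightly varies its pitch.
# Several waves pass through single-input Mixers that trim or boost their individual volume.
# Playback uses a less-positional one-shot attenuation with a wide stereo spread, with bus and submix sends disabled.
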SoundCue /Game/Sounds/Artemis/POI/Hollow_Base_Takover/Spots/Arcade/Spot_Arcade_Stingers_Distant_Cue
root=Modulator_0
nodes [
 {id=Mixer_0, children [WavePlayer_29], properties {inputVolume [0.75]}}
 {id=Mixer_1, children [WavePlayer_22], properties {inputVolume [0.4]}}
 {id=Mixer_2, children [WavePlayer_23], properties {inputVolume [0.6]}}
 {id=Mixer_3, children [WavePlayer_21], properties {inputVolume [1.2]}}
 {id=Mixer_4, children [WavePlayer_28], properties {inputVolume [0.75]}}
 {id=Modulator_0, children [Random_2], properties {pitchMax=1.1, pitchMin=0.9}}
 {
  id=Random_2
  children [
   WavePlayer_17
   WavePlayer_18
   WavePlayer_20
   Mixer_3
   Mixer_1
   Mixer_2
   WavePlayer_24
   WavePlayer_25
   WavePlayer_27
   Mixer_4
   Mixer_0
   WavePlayer_30
  ]
  properties {weights [1,1,1,1.1,1,1,1,1,1,1,1,1]}
 }
 {id=WavePlayer_17, properties {soundWaveAssetPtr=&/Spot_Arcade_Stingers_Distant_01}}
 {id=WavePlayer_18, properties {soundWaveAssetPtr=&/Spot_Arcade_Stingers_Distant_02}}
 {id=WavePlayer_20, properties {soundWaveAssetPtr=&/Spot_Arcade_Stingers_Distant_05}}
 {id=WavePlayer_21, properties {soundWaveAssetPtr=&/Spot_Arcade_Stingers_Distant_06}}
 {id=WavePlayer_22, properties {soundWaveAssetPtr=&/Spot_Arcade_Stingers_Distant_07}}
 {id=WavePlayer_23, properties {soundWaveAssetPtr=&/Spot_Arcade_Stingers_Distant_08}}
 {id=WavePlayer_24, properties {soundWaveAssetPtr=&/Spot_Arcade_Stingers_Distant_09}}
 {id=WavePlayer_25, properties {soundWaveAssetPtr=&/Spot_Arcade_Stingers_Distant_10}}
 {id=WavePlayer_27, properties {soundWaveAssetPtr=&/Spot_Arcade_Stingers_Distant_12}}
 {id=WavePlayer_28, properties {soundWaveAssetPtr=&/Spot_Arcade_Stingers_Distant_13}}
 {id=WavePlayer_29, properties {soundWaveAssetPtr=&/Spot_Arcade_Stingers_Distant_14}}
 {id=WavePlayer_30, properties {soundWaveAssetPtr=&/Spot_Arcade_Stingers_Distant_03}}
]
properties {
 attenuationSettings=/Game/Sounds/Attenuation/Fort_TileAmbient_OneShot_LessPositional_Log_Att
 bEnableBusSends=false
 bEnableSubmixSends=false
 maxDistance=6200
 soundClassObject=&Fort_Ambient_Outdoors_Athena
 volumeMultiplier=1.504
 attenuationOverrides {falloffDistance=6000, stereoSpread=10000, attenuationShapeExtents {x=2000}}
}
